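# Background material for the IWD 2023 item shop takeover, feeding colour into the MP item shop environment setup function.
# Named reroutes build a flickering starfield, purple space clouds and gradient colour waves in a line mask.
# Panned blurred noise distorts gradient UVs and drives a flowmap noise, while a scaled circle SDF supplies cloud frame masks.
# The final colour lerps two rerouted layers by a rerouted mask, attenuated by an inverted top-bottom gradient.
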Material /ItemShop/UI/Art/IWDTakeover/IWD2023/M_UI_IWD23_ShopBG_MP
graph {
 properties {pixelDepthOffsetMode=PDOM_Legacy, usedShadingModels=Default Lit}
 nodes [
  n1 = MaterialFunctionCall!(UVs=NamedRerouteUsage(), Subtract=Constant() {r=0.4}, Falloff=Constant() {r=0.1}, Intensity=Constant() {r=60}) {materialFunction=/Game/UI/Frontend/ItemShop/Art/_MaterialFunctions_ShopTheming/MF_UI_TopBottomGradient}
  n2 = NamedRerouteUsage()
  n3 = NamedRerouteUsage()
  n4 = NamedRerouteUsage()
  MaterialFunctionCall!(
   Color=Multiply(A=OneMinus(None=@n1.Result), B=LinearInterpolate(A=@n2, B=@n3, Alpha=@n4))
   Color Brightness=Constant() {r=0.8}
   Floor Radius
   Floor Density
   Floor Metallic
   Floor Specular
   Floor Roughness
  ) {materialFunction=/Game/UI/Foundation/Environment/Shell/Materials/MaterialFunctions/MF_UI_MPItemShop_Setup}
  MaterialFunctionCall() {materialFunction=&GetUserInterfaceUV}
  n5 = MaterialFunctionCall!(ScreenUV, TextureSize, ScalePivotPoint) {materialFunction=/FortUILibrary/MaterialFunctions/Basics/MF_UI_ScreenSpaceUVScaling}
  NamedRerouteDeclaration(None=@n5.Scaled to Fill) {name=UVScaling_FIll, nodeColor=<rgba 1 0.6563 0 1>}
  n6 = NamedRerouteUsage()
  n7 = Time()
  n8 = TextureSample(UVs=Rotator(Coordinate=Multiply(A=@n6) {constB=2}, Time=Multiply(A=@n7) {constB=0.009})) {constMipValue=1, samplerType=SAMPLERTYPE_LinearColor, texture=/Game/Packages/Fortress_Sky/SkyDome/Master/T_SkyDome_StarCloudMask}
  n9 = ComponentMask(None=@n8.R) {r=true}
  n10 = Multiply(A=@n6) {constB=0.85}
  n11 = ScalarParameter() {defaultValue=0.015, group=Stars, parameterName=StarFlickerSpeed}
  n12 = Subtract(
   A=@n9
   B=TextureSample(UVs=Panner(Coordinate=@n10, Speed=@n11)) {samplerType=SAMPLERTYPE_LinearColor, texture=&T_SkyDome_StarCloudMask}.R
  )
  NamedRerouteDeclaration(None=Multiply(A=Multiply(A=@n12, B=@n9), B=Constant() {r=12})) {name=Stars_1, nodeColor=<rgba 0.8414 0 1 1>}
  NamedRerouteDeclaration(None=@n5.Scaled to Fit) {name=UVScaling_Fit, nodeColor=<rgba 1 0.9844 0 1>}
  n13 = MaterialFunctionCall!(UV, Seed) {materialFunction=&MF_UI_Random}
  n14 = MaterialFunctionCall() {materialFunction=&ScreenAlignedUVs}
  n15 = MaterialFunctionCall() {materialFunction=&ScreenResolution}
  n16 = Panner(
   Coordinate=Multiply(A=LinearInterpolate(A=@n14."X 100%, Y 100%", B=@n15.Visible Resolution) {constAlpha=0.125}, B=Constant() {r=0.3})
   Speed=Constant2Vector() {g=0.005, r=0.002}
  ) {speedX=0.005, speedY=0.05}
  n17 = NamedRerouteUsage()
  n18 = ScalarParameter() {defaultValue=0.3, parameterName=Tiling}
  n19 = TextureSample(UVs=Panner(Coordinate=Multiply(A=@n17, B=@n18), Speed=Constant2Vector() {g=0.0013, r=0.0025}) {speedX=0.025, speedY=0.013}) {samplerType=SAMPLERTYPE_Grayscale, texture=/FortUILibrary/UIKit/Buttons/Textures/LowResBlurredNoise}
  n20 = Multiply(A=Add(
    A=TextureSample(UVs=@n16) {samplerType=SAMPLERTYPE_Grayscale, texture=&LowResBlurredNoise}.RGB
    B=@n19.RGB
   )) {constB=3}
  n21 = TextureSample(UVs=Multiply(A=NamedRerouteUsage(), B=ComponentMask(None=@n20) {g=true, r=true}) {constB=4}) {texture=&/T_UI_IWD23Gradient}
  n22 = Multiply(A=NamedRerouteUsage()) {constB=0.75}
  n23 = SmoothStep(Value=TextureSample(UVs=Panner(Coordinate=@n22) {speedX=0.15, speedY=0.01}) {samplerType=SAMPLERTYPE_Grayscale, texture=&LowResBlurredNoise}.R) {constMax=0.75, constMin=0.25}
  n24 = Add(A=@n23, B=@n21.RGB)
  n25 = LinearInterpolate(A=@n21.RGB, B=Multiply(A=Multiply(A=@n24, B=@n21.RGB)) {constB=1.1}, Alpha=@n23)
  n26 = NamedRerouteUsage()
  n27 = Multiply(A=@n26) {constB=2.75}
  n28 = SmoothStep(Value=@n26)
  n29 = DotProduct(
   A=Subtract(A=LinearInterpolate(A=@n27, B=@n28), B=ScalarParameter() {defaultValue=0.15, parameterName=Line Screen Height})
   B=Constant2Vector() {g=0.95, r=0.95}
  )
  n30 = Multiply(A=MaterialFunctionCall!(Value=Saturate(None=@n29), Scale=Constant() {r=1}, Height) {materialFunction=/FortUILibrary/MaterialFunctions/Basics/MF_UI_PingPong}.Result) {constB=0.6}
  n31 = NamedRerouteDeclaration(None=SmoothStep(Value=Saturate(None=@n30)) {constMax=0.2}) {name=LineMask, nodeColor=<rgba 1 0.9141 0 1>}
  NamedRerouteDeclaration(None=Add(A=Multiply(A=@n13.Result) {constB=0.05}, B=Multiply(A=@n25, B=@n31))) {name=Color Waves, nodeColor=<rgba 0.4188 0 1 1>}
  n32 = NamedRerouteUsage()
  n33 = TextureSample(UVs=Multiply(A=NamedRerouteUsage()) {constB=1.5}) {texture=/Game/UI/Frontend/ItemShop/Art/IWDTakeover/Texture/T-Noise-IWD}
  n34 = SmoothStep(Value=@n33.RGB) {constMax=0.5}
  n35 = Constant3Vector() {constant=<rgba 0.4523 0 1 1>}
  n36 = Constant3Vector() {constant=<rgba 0.7613 0.1208 1 1>}
  n37 = LinearInterpolate(
   A=Constant3Vector() {constant=<rgba 0.0106 0 0.0208 1>}
   B=Multiply(A=Max(A=@n32, B=@n34), B=LinearInterpolate(A=@n35, B=@n36, Alpha=@n33.G))
   Alpha=NamedRerouteUsage()
  )
  NamedRerouteDeclaration(None=Add(A=NamedRerouteUsage(), B=Multiply(A=@n37) {constB=0.2})) {name=Space Clouds, nodeColor=<rgba 0.4213 0 1 1>}
  n38 = NamedRerouteUsage()
  n39 = TextureObject() {samplerType=SAMPLERTYPE_Grayscale, texture=&LowResBlurredNoise}
  n40 = NamedRerouteDeclaration(None=@n16) {name=PanningUV, nodeColor=<rgba 1 0.9844 0 1>}
  n41 = NamedRerouteDeclaration(None=AppendVector(A=@n19.R, B=TextureSample(UVs=@n40) {samplerType=SAMPLERTYPE_Grayscale, texture=&LowResBlurredNoise}.R)) {name=AdriFlowmapUVs_2, nodeColor=<rgba 0 1 0.8125 1>}
  n42 = TextureSample(UVs=@n41, Tex=@n39)
  n43 = OneMinus(None=@n41)
  n44 = Constant() {r=0.1}
  n45 = Constant() {r=0.2}
  NamedRerouteDeclaration(None=MaterialFunctionCall("Value (0-1)"=MaterialFunctionCall!(UVs=@n38, Texture Object=@n39, Grayscale Flow Map=@n42.RGB, Flow Direction=@n43, Flow Intensity=@n44, Flow Speed=@n45) {materialFunction=/FortUILibrary/MaterialFunctions/Effects/MF_UI_GrayscaleFlowMap_New}.Distorted Texture) {materialFunction=&MF_UI_EaseCurves}.EaseInOutSine) {name=Noise, nodeColor=<rgba 0 0.5312 1 1>}
  n46 = NamedRerouteUsage()
  n47 = Constant() {r=1}
  n48 = MaterialFunctionCall(
   UVs=MaterialFunctionCall!(
    Texture Scale=Constant3Vector() {constant=<rgba 0.996 0.5 0.5 0>}
    UVs=MaterialFunctionCall(UVs=@n46, Texture Scale=@n47) {materialFunction=&ScaleUVsByCenter}.UVs
    Offset XY Texturesize Z=Constant3Vector() {constant=<rgba 0.264 0.12 0.272 0>}
   ) {materialFunction=/Engine/Functions/Engine_MaterialFunctions01/Texturing/ScaleUVsAroundPoint}.UVs
   Size=Constant() {r=0.9}
   GlowMin=Constant() {r=-1}
   GlowMax=Constant() {r=0.5}
  ) {materialFunction=&MF_UI_Circle}
  NamedRerouteDeclaration(None=@n48.Fill) {name=CloudFrameFill, nodeColor=<rgba 0 1 0.5781 1>}
  NamedRerouteDeclaration(None=SmoothStep(Value=@n48.SDF) {constMax=1.25, constMin=0.5}) {name=CloudFrameSDF, nodeColor=<rgba 0 0.2266 1 1>}
 ]
}
